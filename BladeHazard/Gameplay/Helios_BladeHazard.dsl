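; Spinning blade hazard that damages and launches players and vehicles, self-destroying on authority if disabled by hotfix.
; Outer-mesh overlap hits immediately and starts a per-actor looping timer that re-hits all overlapping actors until overlap ends.
; Each hit traces from the blades to derive a hotfix-tuned launch velocity, then applies environmental damage and a gameplay cue.
; Vehicle impulses scale with speed and are multicast because vehicles are client authoritative.
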(event EventOnWorldReady
  (Networking|SwitchHasAuthority
    (:Authority
      (if (not (ScalableFloat|GetValueasBool (Variables|Hotfix|GetHF_Enabeld)))
        (Actor|DestroyActor)))))

(event OnComponentBeginOverlap(CollisionMeshOuter)
  (params (OverlappedComponent :type Primitive%20Component%20Object%20Reference) (OtherActor :type Actor%20Object%20Reference) (OtherComp :type Primitive%20Component%20Object%20Reference) (OtherBodyIndex :type Integer) (bFromSweep :type Boolean) (SweepResult :type Hit%20Result%20Structure%20%28by%20ref%29))
  (CallFunction|DamageOverlappingActor OtherActor)
  (CallFunction|StartDamageTimerForActor OtherActor))

(event OnComponentEndOverlap(CollisionMeshOuter)
  (params (OverlappedComponent :type Primitive%20Component%20Object%20Reference) (OtherActor :type Actor%20Object%20Reference) (OtherComp :type Primitive%20Component%20Object%20Reference) (OtherBodyIndex :type Integer))
  (CallFunction|ClearDamageTimerForActor OtherActor))

(event Custom|MulticastLaunchVehicle
  (params (VehicleToLaunch :type Fort%20Vehicle%20Interface%20Interface) (LinearVelocity :type Vector%20%28by%20ref%29))
  ;[1 0.4724 0] Multicast version of Impulse vehicle, we need to call this in cliients as vehicles are client authoritative
  (Vehicle|ImpulseVehicle VehicleToLaunch LinearVelocity (Math|Vector|VectorZero) false false true "None"))

(fn ApplyEnvDamage
  (params (CurrentHitActor :type Actor%20Object%20Reference) (HitResult :type Hit%20Result%20Structure%20%28by%20ref%29))
  (bind _sf_playerdamage (Variables|Hotfix|GetSF_PlayerDamage))
  (bind _returnvalue (ScalableFloat|GetValueatLevel _sf_playerdamage))
  (bind _returnvalue_2 (Utilities|DoesObjectImplementInterface CurrentHitActor "/Script/FortniteGame.FortVehicleInterface"))
  (bind _sf_vehicledamage (Variables|Hotfix|GetSF_VehicleDamage))
  (bind _returnvalue_3 (ScalableFloat|GetValueatLevel _sf_vehicledamage))
  (bind _returnvalue_4 (Math|Float|SelectFloat _returnvalue_3 _returnvalue _returnvalue_2))
  (if (> _returnvalue_4 0)
    (Environment|ApplyGlobalEnvironmentDamagetoActor (Variables|GetCurrentHitActor) _returnvalue_4 (GameplayTags|GetOwnedGameplayTags :!class /Script/GameplayTags.GameplayTagAssetInterface) (select (Utilities|DoesObjectImplementInterface (Variables|GetCurrentHitActor) "/Script/FortniteGame.FortVehicleInterface") "/BladeHazard/Gameplay/GameplayEffects/GE_BladesEnvironmentalDamage_Vehicle.GE_BladesEnvironmentalDamage_Vehicle_C" "/BladeHazard/Gameplay/GameplayEffects/GE_BladesEnvironmentalDamage_Player.GE_BladesEnvironmentalDamage_Player_C"))
    (Ability|ExecuteGameplayCueWithParamsReplicated (Variables|Default|GetHit%20Vehicle%20Gameplay%20Cue%20Tag) (GameplayCue|MakeGameplayCueParametersfromHitResult (Variables|GetHitResult)))))

(fn ClearDamageTimerforActor
  (params (Actor :type Actor%20Object%20Reference%20%28by%20ref%29))
  (bind _actortotimermap (Variables|Default|GetActorToTimerMap))
  (bind (_value _returnvalue) (Utilities|Map|Find _actortotimermap Actor))
  (if _returnvalue
    (Utilities|Time|ClearandInvalidateTimerbyHandle _value)
    (Utilities|Map|Remove (Variables|Default|GetActorToTimerMap) (Variables|GetActor))))

(fn DamageOverlappingActor
  (params (HitActor :type Actor%20Object%20Reference))
  (bind _outhit (CallFunction|TracefromBladestoHitActorforHitResult HitActor))
  (Default|OnBladeHitActor HitActor _outhit))

(fn DamageOverlappingActors
  (for _array_element (Collision|GetOverlappingActors (Variables|Default|GetCollisionMeshOuter) 0 :!class /Script/Engine.PrimitiveComponent)
    (CallFunction|DamageOverlappingActor _array_element)))

(fn GetForceVelocityforHit
  (params (Hit :type Hit%20Result%20Structure) (HitActor :type Actor%20Object%20Reference))
  (bind (_bblockinghit _binitialoverlap _time _distance _location _impactpoint _normal) (Collision|BreakHitResult Hit))
  (return (* (+ (Utilities|Operators|Multiply (Math|Vector|SelectVector "0.000000,0.000000,1.000000" (CallFunction|GetForceDirectionforHit Hit) (and (Utilities|Operators|Equal(==) :A (CallFunction|GetForceDirectionforHit Hit)) (ScalableFloat|GetValueasBool (Variables|Hotfix|GetHF_HackVehicleImpulse)))) (ScalableFloat|GetValueatLevel (Variables|Hotfix|GetHF_TangentVelocityStrengthPercent)) (CallFunction|GetRepellingForceMagnitudeBasedonHitActor HitActor)) (CallFunction|GetForwardForceModifier _normal HitActor)) (Math|Float|SelectFloat 0.5 (Math|Float|SelectFloat (ScalableFloat|GetValueatLevel (Variables|Hotfix|GetHF_VehicleForceStrengthPercent)) (ScalableFloat|GetValueatLevel (Variables|Hotfix|GetHF_PlayerForceStrengthPercent)) (Utilities|DoesObjectImplementInterface (Variables|GetHitActor) "/Script/FortniteGame.FortVehicleInterface")) (and (Utilities|DoesObjectImplementInterface (Variables|GetHitActor) "/Script/FortniteGame.FortVehicleInterface") (and (Utilities|Operators|Equal(==) :A (CallFunction|GetForceDirectionforHit Hit)) (ScalableFloat|GetValueasBool (Variables|Hotfix|GetHF_HackVehicleImpulse))))))))

(fn GetForwardForceModifier
  (params (HitNormal :type Vector) (HitActor :type Actor%20Object%20Reference))
  (return (Math|Vector|SelectVector (+ (* (Utilities|Select :Index (Variables|FortBladeHazard|GetSpinningDirection) :Clockwise (Utilities|Operators|Multiply HitNormal (ScalableFloat|GetValueatLevel (Variables|Hotfix|GetHF_ForwardVelocityStrengthPercent)) (CallFunction|GetRepellingForceMagnitudeBasedonHitActor HitActor)) :Counterclockwise (Utilities|Operators|Multiply HitNormal (ScalableFloat|GetValueatLevel (Variables|Hotfix|GetHF_ForwardVelocityStrengthPercent)) (CallFunction|GetRepellingForceMagnitudeBasedonHitActor HitActor)) :SnapToCustomDirections "0, 0, 0") -1.0) (+ (Transformation|GetActorUpVector (Variables|GetHitActor)) (ScalableFloat|GetValueatLevel (Variables|Hotfix|GetHF_VehicleExtraZAddedToBladeHits)))) (+ (* (Utilities|Select :Index (Variables|FortBladeHazard|GetSpinningDirection) :Clockwise (Utilities|Operators|Multiply HitNormal (ScalableFloat|GetValueatLevel (Variables|Hotfix|GetHF_ForwardVelocityStrengthPercent)) (CallFunction|GetRepellingForceMagnitudeBasedonHitActor HitActor)) :Counterclockwise (Utilities|Operators|Multiply HitNormal (ScalableFloat|GetValueatLevel (Variables|Hotfix|GetHF_ForwardVelocityStrengthPercent)) (CallFunction|GetRepellingForceMagnitudeBasedonHitActor HitActor)) :SnapToCustomDirections "0, 0, 0") -1.0) (+ (Transformation|GetActorUpVector (Variables|GetHitActor)) (ScalableFloat|GetValueatLevel (Variables|Hotfix|GetHF_PlayerExtraZAddedToBladeHits)))) (Utilities|DoesObjectImplementInterface (Variables|GetHitActor) "/Script/FortniteGame.FortVehicleInterface"))))

(fn GetRepellingForceMagnitudeBasedOnHitActor
  (params (HitActor :type Actor%20Object%20Reference))
  (return (Math|Float|SelectFloat (ScalableFloat|GetValueatLevel (Variables|Hotfix|GetHF_VehicleRepellingForceMagnitude)) (ScalableFloat|GetValueatLevel (Variables|Hotfix|GetHF_PlayerRepellingForceMagnitude)) (Utilities|DoesObjectImplementInterface HitActor "/Script/FortniteGame.FortVehicleInterface"))))

(fn GetVehicleSpeedPercent
  (params (FortVehicleInterface :type Fort%20Vehicle%20Interface%20Interface))
  ;[1 1 1] We don't have a way to get reverse percent, so if we have a negative speed, divide by a generally higher top forward speed and clamp to 1.0 max
  ; SUV backing at full speed will be 800+ or 0.5+ percent in this case
  (return (Math|Float|Clamp(Float) (Math|Float|SelectFloat (Utilities|ClampMaxFloat (Math|Float|SafeDivide (Math|Float|Absolute(Float) (Vehicle|GetForwardSpeed FortVehicleInterface)) 1.5e+03) 1.0) (Vehicle|GetForwardSpeedPercent FortVehicleInterface) (< (Vehicle|GetForwardSpeed FortVehicleInterface) 0)) (ScalableFloat|GetValueatLevel (Variables|Hotfix|GetHF_VehicleMinClampSpeedPercent)))))

(fn MakeApproximateHitfromActor
  (params (Actor :type Actor%20Object%20Reference))
  (return (Collision|MakeHitResult true false 0.0 0.0 (Transformation|GetActorLocation) "0, 0, 0" (Math|Vector|Normalize (- (Transformation|GetActorLocation Actor) (Transformation|GetActorLocation))))))

(fn OnBladeHitActor
  (params (HitActor :type Actor%20Object%20Reference) (Hit :type Hit%20Result%20Structure))
  (bind _returnvalue (Default|GetForceVelocityForHit Hit HitActor))
  (bind _asfort_vehicle_interface (Utilities|Casting|CastToFortVehicleInterface HitActor)
    (:then
      (CallFunction|MulticastLaunchVehicle _asfort_vehicle_interface (* _returnvalue (CallFunction|GetVehicleSpeedPercent _asfort_vehicle_interface))))
    (:CastFailed
      (bind _asfort_player_pawn (Utilities|Casting|CastToFortPlayerPawn (Variables|GetHitActor)))
      (if (not (Vehicle|IsinVehicle _asfort_player_pawn))
        (Character|LaunchCharacter _asfort_player_pawn _returnvalue false false))))
  (Default|ApplyEnvDamage (Variables|GetHitActor) Hit))

(fn SpawnHitFX
  (params (TargetLocation :type Vector) (HitRotationVelocity :type Vector) (Hit%20FX :type Niagara%20System%20Object%20Reference))
  (bind _returnvalue (Niagara|SpawnSystematLocation :SystemTemplate Hit%20FX :Location (+ (Math|Vector|Lerp(Vector) TargetLocation (Transformation|GetActorLocation) 0.4) "0.000000,0.000000,60.000000") :Scale "1.000000,1.000000,1.000000" :bAutoDestroy true :bAutoActivate true :PoolingMethod "AutoRelease" :bPreCullCheck true))
  (Niagara|SetNiagaraVariable(Vector3) _returnvalue "HitRotationVelocity" HitRotationVelocity))

(fn StartDamageTimerforActor
  (params (Actor :type Actor%20Object%20Reference%20%28by%20ref%29))
  (CallFunction|ClearDamageTimerForActor Actor) ; Clear and remove any existing timers we could have for\nthis actor (should not happen in the normal case)
  (bind _returnvalue (Utilities|Time|SetTimerbyEvent (EventDispatchers|CreateEvent :!event DamageOverlappingActors) (ScalableFloat|GetValueatLevel (Variables|Hotfix|GetHF_DamageInterval)) true true))
  (Utilities|Map|Add (Variables|Default|GetActorToTimerMap) (Variables|GetActor) _returnvalue))

(fn TraceFromBladesToHitActorForHitResult
  (params (HitActor :type Actor%20Object%20Reference))
  (bind _outhit (Collision|LineTraceForObjects :Start (+ (Transformation|GetActorLocation self) (LocalVariables|GetBladesHeightOffsetTraceStart)) :End (+ (Transformation|GetActorLocation HitActor) (Math|Vector|SelectVector (LocalVariables|GetVehicleHeightOffsetTraceEnd) (LocalVariables|GetPawnHeightOffsetTraceEnd) (Utilities|DoesObjectImplementInterface (Variables|GetHitActor) "/Script/FortniteGame.FortVehicleInterface"))) :ObjectTypes (LocalVariables|GetTraceObjectTypes) :bTraceComplex false :DrawDebugType "ForDuration" :bIgnoreSelf true :TraceColor "(R=1.000000,G=0.000000,B=0.000000,A=1.000000)" :TraceHitColor "(R=0.000000,G=1.000000,B=0.000000,A=1.000000)" :DrawTime 5.0))
  (return _outhit))

(fn ConstructionScript
  (Components|InstancedStaticMesh|AddInstances (Variables|Default|GetISM_blades) (Variables|Default|GetBladeInstanceTransforms) false false true))
